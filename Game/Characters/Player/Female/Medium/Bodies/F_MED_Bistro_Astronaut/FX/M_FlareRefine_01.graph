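# Additive unlit Niagara sprite and mesh particle material for flare effects, stored with the Bistro Astronaut character FX.
# A channel-masked flare texture goes through MF_FX_Refine with a color parameter, is multiplied by particle color, then passes through emissive adjustments.
# Opacity multiplies particle alpha, the masked texture and a flare mask scalar.
# Quality and shading-path switches keep depth fade on Default, High and Epic deferred rendering, skipping it on Low, Medium, forward and mobile.
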Material /Game/Characters/Player/Female/Medium/Bodies/F_MED_Bistro_Astronaut/FX/M_FlareRefine_01
graph {
 properties {
  bUsedWithNiagaraMeshParticles=true
  bUsedWithNiagaraSprites=true
  blendMode=BLEND_Additive
  pixelDepthOffsetMode=PDOM_Legacy
  shadingModel=MSM_Unlit
  usedShadingModels=Unlit
 }
 nodes [
  n1 = ChannelMaskParameter(None=TextureSample(Tex=TextureObjectParameter() {group=Flare, parameterName=Flare Tex, texture=/Game/Effects/Fort_Effects/Textures/Flares/T_Flare_Spikeball}).RGBA) {group=Flare, parameterName=FlareMaskChannel}
  n2 = VectorParameter() {group=Flare, parameterName=Flare Refine, defaultValue=<rgba 1 1 1 0.5>}
  n3 = ParticleColor()
  MaterialFunctionCall(Emissive=Multiply(
    A=MaterialFunctionCall(Texture=@n1.R, RGB=@n2.RGB, A=@n2.RGB) {materialFunction=&MF_FX_Refine}.Result
    B=@n3.RGB
   )) {materialFunction=&MF_FX_EmissiveAdjustments}
  n4 = Multiply(A=@n3.A, B=Multiply(A=@n1.R, B=ScalarParameter() {defaultValue=1, group=Flare, parameterName=Flare Mask}))
  n5 = DepthFade(Opacity=@n4, FadeDistance=ScalarParameter() {parameterName=FadeDistance})
  n6 = QualitySwitch(Default=@n5, Low=@n4, High=@n5, Medium=@n4, Epic=@n5)
  ShadingPathSwitch(Default=@n6, Deferred=@n6, Forward=@n4, Mobile=@n4)
 ]
}
