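# Additive unlit two-sided Niagara sprite and mesh particle material for the Bistro Astronaut character FX, named as a mobile particle decal variant.
# A scrolling grid texture on camera-distance-scaled macro UVs drives a remapped sine pulse.
# The pulse adds to a gradient glow, panned by dynamic parameter along Z or local direction, tinted by particle color into emissive.
# Pulse, mask texture and particle alpha form opacity; separate ramp and quality-switched fresnel branches show no downstream connection.
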Material /Game/Characters/Player/Female/Medium/Bodies/F_MED_Bistro_Astronaut/FX/M_ParticleDecal_Solution_Mobile
graph {
 properties {
  bUsedWithNiagaraMeshParticles=true
  bUsedWithNiagaraSprites=true
  blendMode=BLEND_Additive
  pixelDepthOffsetMode=PDOM_Legacy
  shadingModel=MSM_Unlit
  twoSided=true
  usedShadingModels=Unlit
 }
 nodes [
  n1 = ObjectPositionWS()
  n2 = CameraPositionWS()
  n3 = Multiply(
   A=Multiply(A=ParticleMacroUV(), B=ScalarParameter() {defaultValue=1, parameterName=Scale MacroUVs}) {constB=5}
   B=Divide(A=Distance(A=@n1, B=@n2), B=ScalarParameter() {defaultValue=150, parameterName=CameraDistanceDivide}) {constB=150}
  )
  n4 = VectorParameter() {parameterName=DetialMacroUV, defaultValue=<rgba 1.5 1 0 0>}
  n5 = ComponentMask(None=@n4.RGB) {g=true, r=true}
  n6 = AppendVector(A=@n4.B, B=@n4.B)
  n7 = Time()
  n8 = TextureSampleParameter2D(UVs=Add(A=Multiply(A=@n3, B=@n5), B=Multiply(A=@n6, B=@n7) {constB=2})) {
   parameterName=Detial Texture
   samplerType=SAMPLERTYPE_LinearColor
   texture=/Game/Packages/Fortress_SharedMaterials/Textures/T_Slowfield_GridSquares_01
  }
  n9 = Multiply(A=Time(), B=ScalarParameter() {defaultValue=2, parameterName=MacroUV Sine Time Intensity}) {constB=2}
  n10 = ScalarParameter() {parameterName=No Pulse Sine Control}
  n11 = MaterialFunctionCall!(Sine Phase=Add(A=OneMinus(None=@n8.R), B=LinearInterpolate(A=@n9, Alpha=@n10) {constB=0}), Value 1, Value 2) {materialFunction=/Engine/Functions/Engine_MaterialFunctions02/Utility/Sine_Remapped}
  n12 = Multiply(A=@n11.Remapped Sine, B=ScalarParameter() {defaultValue=0.05, parameterName=DetialMultiply}) {constB=0.05}
  n13 = Divide(A=Subtract(A=ObjectPositionWS(), B=WorldPosition().XYZ), B=ScalarParameter() {defaultValue=100, parameterName=GlowPan UV}) {constB=100}
  n14 = VectorParameter() {desc=Which Direction you would like the Mask to pan, group=Local Space Direction, parameterName=ObjectsVector, defaultValue=<rgba 0 1 0 1>}
  n15 = Transform(None=@n14.RGB) {transformSourceType=TRANSFORMSOURCE_Local}
  n16 = LinearInterpolate(
   A=ComponentMask(None=@n13) {b=true}
   B=DotProduct(A=Normalize(VectorInput=@n15), B=@n13)
   Alpha=ScalarParameter() {group=Local Space Direction, parameterName=UseLocalSpaceDirection, sliderMax=1}
  )
  n17 = DynamicParameter() {paramNames [Param1,Param2,GlowPan,Param4]}
  n18 = Multiply(
   A=TextureSampleParameter2D(UVs=Panner(Coordinate=@n16, Time=@n17.GlowPan) {speedY=3}) {parameterName=GlowPan Texture, texture=/Game/Global/Textures/Default/Masks_and_Gradients/T_Vert_Gradient_Soft}.R
   B=ScalarParameter() {defaultValue=2, parameterName=GlowPanScale}
  ) {constB=1.25}
  n19 = ParticleColor()
  MaterialFunctionCall(Emissive=Multiply(A=Add(A=@n12, B=@n18), B=@n19.RGB)) {materialFunction=&MF_FX_EmissiveAdjustments}
  n20 = TextureSampleParameter2D() {parameterName=Mask Texture, texture=/Game/Global/Textures/Default/Masks_and_Gradients/WM_PinGlow_Sphere}
  Multiply(
   A=Multiply(A=@n19.A, B=Multiply(A=@n11.Remapped Sine, B=@n20.R))
   B=ScalarParameter() {defaultValue=1, group=Blinking, parameterName=Opacity}
  )
  Multiply(B=TextureSample() {texture=/Game/Effects/Fort_Effects/Textures/Ramps/T_Power_Source_Sphere_Ramp}.R) {constA=1}
  n21 = PixelNormalWS()
  n22 = VertexNormalWS()
  n23 = QualitySwitch(Default=@n21, Low=@n22, High=@n21, Medium=@n22, Epic=@n21)
  n24 = Abs(None=DotProduct(A=CameraVectorWS(), B=ShadingPathSwitch(Default=@n23, Deferred=@n23, Forward=@n22, Mobile=@n22)))
  n25 = OneMinus(None=@n24)
  n26 = VectorParameter() {
   group=Base
   parameterName=Mesh Fresnel Fade
   sortPriority=1
   channelNames {a="", b=Power, g=Invert, r=Use Mesh Fresnel Fade}
   defaultValue=<rgba 0 0 5 0>
  }
  Saturate(None=Power(Base=LinearInterpolate(A=@n24, B=@n25, Alpha=@n26.Invert), Exp=@n26.Power))
 ]
}
