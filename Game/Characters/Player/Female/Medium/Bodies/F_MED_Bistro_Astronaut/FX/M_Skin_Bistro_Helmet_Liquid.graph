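# Masked, two-sided lit material for the liquid effect on the Bistro Astronaut skin's helmet, built for Niagara mesh particles.
# Scrolling, noise-distorted normal samples drive surface normals and two-tone emissive detail, with separate inside and outside brightness.
# Opacity thresholds noise plus a UV gradient against a dissolve value (scalar or dynamic parameter), multiplied by a smoke-texture cutout.
# A scalar blends emissive and mask toward a translucent-helmet look using cubemap reflection, fresnel and pulsing object-anchored screen-space grid.
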Material /Game/Characters/Player/Female/Medium/Bodies/F_MED_Bistro_Astronaut/FX/M_Skin_Bistro_Helmet_Liquid
graph {
 properties {bUsedWithNiagaraMeshParticles=true, blendMode=BLEND_Masked, pixelDepthOffsetMode=PDOM_Legacy, twoSided=true, usedShadingModels=Default Lit}
 nodes [
  n1 = TwoSidedSign()
  LinearInterpolate(
   A=VectorParameter() {group=Twoside, parameterName=BaseColor Inside, defaultValue=<rgba 0.0573 0.0027 0.0416 1>}.RGB
   B=VectorParameter() {group=Twoside, parameterName=BaseColor Outside, defaultValue=<rgba 0.0573 0.0027 0.0416 1>}.RGB
   Alpha=@n1
  )
  n2 = DynamicParameter() {defaultValue=<rgba 1 1 1 0>}
  n3 = LinearInterpolate(
   A=ScalarParameter() {group=Dissolve, parameterName=Dissolve}
   B=@n2.Param1
   Alpha=ScalarParameter() {group=Dissolve, parameterName=Use Dynamic}
  )
  n4 = VectorParameter() {group=UV, parameterName=NormalUV, defaultValue=<rgba 1 1 0 0>}
  n5 = Multiply(A=TextureCoordinate(), B=ComponentMask(None=@n4.RGB) {g=true, r=true})
  n6 = Multiply(A=AppendVector(A=@n4.B, B=@n4.B), B=Time())
  n7 = VectorParameter() {group=UV, parameterName=NoiseUV, defaultValue=<rgba 1 1 0 0>}
  n8 = Multiply(A=TextureCoordinate(), B=ComponentMask(None=@n7.RGB) {g=true, r=true})
  n9 = Multiply(A=AppendVector(A=@n7.B, B=@n7.B), B=Time())
  n10 = TextureObjectParameter() {
   group=Tex
   parameterName=Detial Normal Noise Tex
   samplerType=SAMPLERTYPE_Normal
   texture=&/T_GaussianNoise_Normal
  }
  n11 = Multiply(
   A=TextureSample(UVs=Add(A=@n8, B=@n9), Tex=@n10) {samplerType=SAMPLERTYPE_Normal, texture=/Game/Packages/Fortress_Sky/Materials/2Normal}.R
   B=ScalarParameter() {defaultValue=0.04, parameterName=NoiseStrength}
  ) {constB=0.04}
  n12 = TextureSample(UVs=Add(A=Add(A=@n5, B=@n6), B=@n11), Tex=@n10)
  n13 = ComponentMask(None=TextureCoordinate()) {g=true}
  n14 = OneMinus(None=@n13)
  n15 = ScalarParameter() {parameterName=Dissolve_Axial}
  n16 = Add(
   A=Multiply(A=@n12.R, B=ScalarParameter() {defaultValue=1, group=Dissolve, parameterName=DissolveTex Intensity})
   B=Multiply(A=ScalarParameter() {defaultValue=1, group=Dissolve, parameterName=DissolveMask Intensity}, B=LinearInterpolate(A=@n14, B=@n13, Alpha=@n15))
  )
  n17 = Constant() {r=1}
  n18 = Constant()
  n19 = TextureCoordinate()
  n20 = VectorParameter() {group=UV, parameterName=DissolveUV, defaultValue=<rgba 1 1 0 0>}
  n21 = ComponentMask(None=@n20.RGB) {g=true, r=true}
  n22 = AppendVector(A=@n20.B, B=@n20.B)
  n23 = Time()
  n24 = TextureSample(
   UVs=Add(A=Multiply(A=@n19, B=@n21), B=Multiply(A=@n22, B=@n23))
   Tex=TextureObjectParameter() {group=Tex, parameterName=DissolveOutTex, texture=&T_Inky_Smoke_Tile}
  )
  n25 = Constant() {r=1}
  n26 = Constant()
  n27 = ScalarParameter() {group=TranslucentHelmet, parameterName=Use TranslucentHelmet Unlit}
  LinearInterpolate(
   A=Multiply(A=If(A=@n3, B=@n16, "A > B"=@n17, "A == B"=@n17, "A < B"=@n18), B=If(A=@n2.Param3, B=@n24.R, "A > B"=@n25, "A == B"=@n25, "A < B"=@n26))
   B=NamedRerouteUsage()
   Alpha=@n27
  )
  n28 = VectorParameter() {group=UV, parameterName=DetialUV, defaultValue=<rgba 1 1 0 0>}
  n29 = Multiply(A=TextureCoordinate(), B=ComponentMask(None=@n28.RGB) {g=true, r=true})
  n30 = Multiply(A=AppendVector(A=@n28.B, B=@n28.B), B=Time())
  n31 = TextureSample(UVs=Add(A=Add(A=@n29, B=@n30), B=@n11), Tex=@n10)
  n32 = Clamp(None=Power(Base=MaterialFunctionCall(f2=@n31.RGB) {materialFunction=&AddComponents}.f2) {constExponent=3})
  n33 = Multiply(
   A=LinearInterpolate(
    A=VectorParameter() {group=Twoside, parameterName=Detial DarkColor, defaultValue=<rgba 0.0144 0 0.0156 1>}.RGB
    B=VectorParameter() {group=Twoside, parameterName=Detial LightColor, defaultValue=<rgba 0.3229 0.0151 0.2293 1>}.RGB
    Alpha=Abs(None=@n32)
   )
   B=ParticleColor().RGB
  )
  n34 = LinearInterpolate(
   A=ScalarParameter() {defaultValue=0.5, group=Twoside, parameterName=InsideMultiply}
   B=ScalarParameter() {defaultValue=2, group=Twoside, parameterName=OutsideMultiply}
   Alpha=@n1
  )
  MaterialFunctionCall(Emissive=LinearInterpolate(A=Multiply(A=@n33, B=@n34), B=NamedRerouteUsage(), Alpha=@n27)) {materialFunction=&MF_FX_EmissiveAdjustments}
  MaterialFunctionCall(Normal=@n12.RGB, Flatness=ScalarParameter() {defaultValue=-4, parameterName=NormalStrength}) {materialFunction=&FlattenNormal}
  n35 = TextureSampleParameterCube!(
   UVs=MaterialFunctionCall!(Normal=ReflectionVectorWS(), CameraVector=Constant3Vector() {constant=<rgba 0.7206 -0.1881 -3.3193 1>}) {materialFunction=/Engine/Functions/Engine_MaterialFunctions02/Utility/CustomReflectionVector}.Reflection Vector
   Level
   Apply View MipBias
  ) {mipValueMode=TMVM_MipLevel, parameterName=CubeMapSmallSizeOnly, texture=/Game/Packages/Fortress_Sky/TexturesHDR/T_hdrsky_test03_low}
  n36 = ParticleColor()
  n37 = ScreenPosition()
  n38 = MaterialFunctionCall(World Position=ObjectPositionWS()) {materialFunction=&TransformToClipSpace}
  n39 = Distance(A=ObjectPositionWS(), B=CameraPositionWS())
  n40 = Multiply(
   A=Multiply(A=Subtract(A=@n37.ViewportUV, B=@n38.Buffer UVs for Scene Textures), B=Divide(A=@n39) {constB=80})
   B=Constant2Vector() {g=1, r=1.5}
  )
  n41 = Add(
   A=OneMinus(None=TextureSample(UVs=@n40) {samplerType=SAMPLERTYPE_LinearColor, texture=/Game/Packages/Fortress_SharedMaterials/Textures/T_Slowfield_GridSquares_01}.R)
   B=Multiply(A=Time()) {constB=2}
  )
  n42 = Multiply(A=Fresnel() {baseReflectFraction=0.1, exponent=1}) {constB=1.5}
  n43 = Multiply(A=Fresnel() {baseReflectFraction=0.01, exponent=10}) {constB=5}
  n44 = Add(
   A=Multiply(
    A=MaterialFunctionCall!(Sine Phase=@n41, Value 1, Value 2) {materialFunction=/Engine/Functions/Engine_MaterialFunctions02/Utility/Sine_Remapped}.Remapped Sine
    B=@n42
   )
   B=@n43
  ) {constB=0.05}
  n45 = ParticleColor()
  NamedRerouteDeclaration(None=Add(A=Multiply(A=@n35.RGB, B=@n36.A), B=Multiply(A=@n44, B=@n45.RGB))) {name=TranslucentHelmet_Emi, nodeColor=<rgba 1 0.5625 0 1>}
  n46 = Add(A=Add(A=@n44, B=@n42), B=Saturate(None=@n43))
  n47 = OneMinus(None=ComponentMask(None=TextureCoordinate()) {g=true})
  n48 = Multiply(A=@n47, B=@n47)
  NamedRerouteDeclaration(None=Multiply(A=Multiply(A=@n46, B=@n48), B=@n45.A)) {name=TranslucentHelmet_Alpha, nodeColor=<rgba 0.4531 1 0 1>}
  ScalarParameter() {defaultValue=0.22, parameterName=Roughness}
 ]
}
