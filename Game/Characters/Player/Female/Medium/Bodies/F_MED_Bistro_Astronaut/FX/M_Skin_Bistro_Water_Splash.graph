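# Masked, lit water splash material for Niagara sprites and ribbons in the Bistro Astronaut skin's FX.
# Two panning noise normals multiply to distort UVs for the splash mask, normal and scrolling caustics tile.
# Subtracting a Step value, optionally from Dynamic Parameters, erodes the mask.
# Use Tile Tex blends in tile-modulated erosion and Fresnel-tinted color; emissive takes particle color, and temporal dither scales a depth threshold.
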Material /Game/Characters/Player/Female/Medium/Bodies/F_MED_Bistro_Astronaut/FX/M_Skin_Bistro_Water_Splash
graph {
 properties {
  bUsedWithNiagaraRibbons=true
  bUsedWithNiagaraSprites=true
  blendMode=BLEND_Masked
  pixelDepthOffsetMode=PDOM_Legacy
  usedShadingModels=Default Lit
 }
 nodes [
  n1 = TextureSample(UVs=Panner(Coordinate=TextureCoordinate()) {speedX=0.25, speedY=-0.1}) {samplerType=SAMPLERTYPE_Normal, texture=&/T_GaussianNoise_Normal}
  n2 = TextureSample(UVs=Panner(Coordinate=TextureCoordinate() {uTiling=0.7, vTiling=0.7}) {speedX=-0.15, speedY=0.15}) {samplerType=SAMPLERTYPE_Normal, texture=&T_GaussianNoise_Normal}
  n3 = Multiply(A=@n1.RGB, B=@n2.RGB)
  n4 = Add(
   A=Multiply(A=ComponentMask(None=@n3) {g=true, r=true}, B=ScalarParameter() {defaultValue=0.2, group=Noise, parameterName=NoiseStrength}) {constB=0.2}
   B=ScalarParameter() {defaultValue=-0.02, group=Noise, parameterName=NoiseOffset}
  ) {constB=-0.02}
  n5 = Add(A=TextureCoordinate(), B=@n4)
  n6 = TextureObjectParameter() {
   group=Splash
   parameterName=SplashTex
   samplerType=SAMPLERTYPE_LinearGrayscale
   texture=&/T_Water_Splash_Alpha_02
  }
  n7 = ChannelMaskParameter(None=TextureSample(UVs=@n5, Tex=@n6).RGBA) {group=Splash, parameterName=SplashMaskChannel}
  n8 = DynamicParameter() {defaultValue=<rgba 0 1 1 1>, paramNames [Step,BaseColor,Dissolve,Param4]}
  n9 = LinearInterpolate(A=ScalarParameter() {parameterName=Step}, B=@n8.Step, Alpha=ScalarParameter() {parameterName=Use Dynamic Step})
  n10 = VectorParameter() {group=Tile, parameterName=TileUV, defaultValue=<rgba 1.5 1.5 0.005 0.02>}
  n11 = Multiply(A=TextureCoordinate(), B=AppendVector(A=@n10.R, B=@n10.R))
  n12 = Multiply(A=AppendVector(A=@n10.B, B=@n10.B), B=Time())
  n13 = TextureObjectParameter() {
   group=Tile
   parameterName=TileTex
   samplerType=SAMPLERTYPE_LinearGrayscale
   texture=/Game/Global/Textures/Default/Caustics/T_Nosie_Caustics_Grayscale
  }
  n14 = TextureSample(UVs=Add(A=Add(A=@n11, B=@n12), B=@n4), Tex=@n13) {samplerType=SAMPLERTYPE_LinearGrayscale, texture=&T_Nosie_Caustics_Grayscale}
  LinearInterpolate(
   A=Subtract(A=@n7.R, B=@n9)
   B=Multiply(A=@n7.R, B=Subtract(A=@n14.R, B=@n9))
   Alpha=ScalarParameter() {group=Tile, parameterName=Use Tile Tex}
  )
  n15 = LinearInterpolate(
   A=Multiply(A=@n14.R, B=VectorParameter() {group=Tile, parameterName=Tile Color, defaultValue=<rgba 1 0.0208 0.6083 0>}.RGB)
   B=VectorParameter() {group="Fresnel ", parameterName=Fresnel Color, defaultValue=<rgba 0.8122 0.1399 1 0>}.RGB
   Alpha=Fresnel(
    ExponentIn=ScalarParameter() {defaultValue=6, group="Fresnel ", parameterName=Fresnel ExponentIn}
    BaseReflectFractionIn=ScalarParameter() {defaultValue=0.02, group="Fresnel ", parameterName=Fresnel BaseReflectFractionIn}
   )
  )
  LinearInterpolate(
   A=Multiply(A=VectorParameter() {parameterName=BaseColor, defaultValue=<rgba 1 1 1 1>}.RGB, B=@n8.BaseColor)
   B=@n15
   Alpha=ScalarParameter() {group=Tile, parameterName=Use Tile Tex}
  )
  n16 = TextureObjectParameter() {
   group=Splash
   parameterName=NormalTex
   samplerType=SAMPLERTYPE_Normal
   texture=&/T_Water_Splash_Normal_02
  }
  MaterialFunctionCall(Normal=TextureSample(UVs=@n5, Tex=@n16).RGB, Flatness=ScalarParameter() {defaultValue=-1, group=Splash, parameterName=NormalStrength}) {materialFunction=&FlattenNormal}
  n17 = LinearInterpolate(
   A=VectorParameter() {parameterName=Emissive Dark Color, defaultValue=<rgba 1 1 1 1>}.RGB
   B=VectorParameter() {parameterName=Emissive Light Color, defaultValue=<rgba 1 1 1 1>}.RGB
   Alpha=Power(Base=@n7.R, Exp=ScalarParameter() {defaultValue=1, group=Splash, parameterName=Emissive Power})
  )
  n18 = ScalarParameter() {group=Tile, parameterName=Use Tile Tex}
  MaterialFunctionCall(Emissive=Multiply(A=LinearInterpolate(A=@n17, B=@n15, Alpha=@n18), B=ParticleColor().RGB)) {materialFunction=&MF_FX_EmissiveAdjustments}
  Multiply(
   A=ScalarParameter() {parameterName=DitherDepthThershold}
   B=MaterialFunctionCall() {materialFunction=&DitherTemporalAA}.Result
  )
  ScalarParameter() {defaultValue=0.2, parameterName=Roughness}
 ]
}
