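# Material function for the Bistro Astronaut character's transition effect, layering an animated emissive colour over incoming material attributes.
# Pre-skinned local position, scaled and time-panned, gives a single XZ planar projection (despite "TriPlanner" parameter names) into a gradient texture that blends two emissive colours.
# The BistroCurve3 parameter blends that result toward TransitionColor.
# The Emissive Alpha input masks the glow and darkens base colour; mask and colour are also output.
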MaterialFunction /Game/Characters/Player/Female/Medium/Bodies/F_MED_Bistro_Astronaut/Materials/MF_BistroTransition
graph {nodes [
  n1 = GetMaterialAttributes(None=FunctionInput() {inputType=FunctionInput_MaterialAttributes}) {attributeGetTypes [#1, #2]}
  n2 = VertexInterpolator(VS=PreSkinnedPosition())
  n3 = Multiply(
   A=ScalarParameter() {defaultValue=1, group=PanningEmissive, parameterName=Emissive Tiling}
   B=Divide(
    A=FunctionInput(Preview=@n2.PS) {bUsePreviewValueAsDefault=true, inputName=Local Position}
    B=ScalarParameter() {defaultValue=40, group=PanningEmissive, parameterName=Triplanner_ConstantScaleDivisor}
   )
  )
  n4 = Frac(None=Multiply(A=Time(), B=VectorParameter() {group=PanningEmissive, parameterName=TriPlanner Speed, defaultValue=<rgba 0 0.2 0.2 1>}.RGB))
  n5 = Add(B=ComponentMask(None=Add(A=@n3, B=@n4)) {b=true, r=true})
  n6 = TextureObject() {samplerType=SAMPLERTYPE_Masks, texture=/Game/Characters/Player/Female/Medium/Bodies/F_MED_Bistro_Astronaut/Textures/BistroGrad_Gradient}
  n7 = LinearInterpolate(
   A=LinearInterpolate(
    A=VectorParameter() {parameterName=EmissiveColor A, defaultValue=<rgba 0 10 1.0881 1>}.RGB
    B=VectorParameter() {parameterName=EmissiveColor B, defaultValue=<rgba 0 10 9.8109 1>}.RGB
    Alpha=TextureSample(UVs=@n5, Tex=@n6) {samplerType=SAMPLERTYPE_Masks, texture=/Game/Weapons/WeaponSkins/Textures/T_Swirly_Noise_highContrast}.RGB
   )
   B=VectorParameter() {parameterName=TransitionColor, defaultValue=<rgba 6.5426 0 10 1>}.RGB
   Alpha=Saturate(None=ScalarParameter() {defaultValue=1, parameterName=BistroCurve3})
  )
  n8 = FunctionInput() {inputName=Emissive Alpha}
  FunctionOutput(None=SetMaterialAttributes!(
    MaterialAttributes=@n1.MaterialAttributes
    EmissiveColor=Multiply(A=@n7, B=@n8)
    BaseColor=LinearInterpolate(A=@n1.BaseColor, Alpha=@n8) {constB=0}
   ) {attributeSetTypes [#1, #2]})
  FunctionOutput(None=@n8) {outputName=Emissive Mask}
  FunctionOutput(None=@n7) {outputName=Emissive Color}
 ]}
