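# Material function for the Bistro Astronaut character's ray gun effect, adding a panning star glow to incoming material attributes.
# Two world-aligned projections of a star mask pan separately, the first fed pre-skinned position, and their maximum is tinted and added to a primary color.
# A powered Fresnel input fades that glow to black before it joins Emissive.
# Specular is also overridden, and the Alpha input blends the result with the original attributes.
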MaterialFunction /Game/Characters/Player/Female/Medium/Bodies/F_MED_Bistro_Astronaut/Materials/MF_RayGunMagic
graph {nodes [
  n1 = FunctionInput() {inputType=FunctionInput_MaterialAttributes}
  n2 = GetMaterialAttributes(None=@n1) {attributeGetTypes [#1]}
  n3 = VectorParameter() {group=Gun, parameterName=Ray_Primary, defaultValue=<rgba 0 0 0 1>}
  n4 = TextureObject() {samplerType=SAMPLERTYPE_Masks, texture=/Game/Characters/Player/Female/Medium/Bodies/F_MED_Bistro_Astronaut/Textures/Star_M}
  n5 = ScalarParameter() {defaultValue=10, group=Gun, parameterName=StarSize}
  n6 = FunctionInput(Preview=VertexInterpolator(VS=PreSkinnedPosition()).PS) {inputName=pre Skinned}
  n7 = AppendVector(A=ScalarParameter() {group=Gun, parameterName=Pan X}, B=ScalarParameter() {group=Gun, parameterName=Pan Y})
  n8 = ScalarParameter() {group=Gun, parameterName=Contrast}
  n9 = ScalarParameter() {defaultValue=10, group=Gun, parameterName=StarSize2}
  n10 = AppendVector(A=ScalarParameter() {group=Gun, parameterName=Pan X2}, B=ScalarParameter() {group=Gun, parameterName=Pan Y2})
  n11 = ScalarParameter() {group=Gun, parameterName=Contrast2}
  n12 = Multiply(
   A=VectorParameter() {group=Gun, parameterName=Star Color, defaultValue=<rgba 1 1 1 1>}.RGB
   B=Max(
    A=MaterialFunctionCall!(TextureObject=@n4, TextureSize=@n5, WorldPosition=@n6, Export Float 4, In=@n7, World Space Normal, ProjectionTransitionContrast=@n8) {materialFunction=/Game/Athena/SafeZone/WorldAlignedTexture_ScalePan}.XY Texture
    B=MaterialFunctionCall(TextureObject=@n4, TextureSize=@n9, In=@n10, ProjectionTransitionContrast=@n11) {materialFunction=&WorldAlignedTexture_ScalePan}.XY Texture
   )
  )
  n13 = Power(
   Base=FunctionInput() {inputName=Fresnel, inputType=FunctionInput_Scalar}
   Exp=ScalarParameter() {defaultValue=2, group=Gun, parameterName=FresnelPower}
  )
  n14 = Add(A=@n2.EmissiveColor, B=LinearInterpolate(A=Add(A=@n3.RGB, B=@n12), Alpha=Saturate(None=@n13)) {constB=0})
  n15 = ScalarParameter() {defaultValue=0.8, parameterName=Spec}
  n16 = FunctionInput() {bUsePreviewValueAsDefault=true, inputName=Alpha, inputType=FunctionInput_Scalar}
  FunctionOutput(None=BlendMaterialAttributes(
    A=@n1
    B=SetMaterialAttributes!(MaterialAttributes=@n2.MaterialAttributes, EmissiveColor=@n14, Specular=@n15) {attributeSetTypes [#1, #2]}
    Alpha=Saturate(None=@n16)
   ))
 ]}
