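# Master skeletal-mesh body material for the Bistro Astronaut female player character, layering special effects between shared character PreFX and PostFX functions.
# The Skin FX texture's channels mask a metals LUT, supply transition emissive alpha and drive RayGunMagic alpha.
# MaterialCurveB blends that emissive alpha toward a TransitionMap, feeding the Bistro transition function.
# The result passes through weapon-skin emissive fresnel and RayGunMagic functions before PostFX.
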Material /Game/Characters/Player/Female/Medium/Bodies/F_MED_Bistro_Astronaut/Materials/M_Bistro_BodyMaster
graph {
 properties {
  bUseMaterialAttributes=true
  bUsedWithMeshDeformer=true
  bUsedWithMorphTargets=true
  bUsedWithSkeletalMesh=true
  pixelDepthOffsetMode=PDOM_Legacy
  usedShadingModels=Default Lit
 }
 nodes [
  n1 = MaterialFunctionCall() {materialFunction=&BaseFN_CharacterFunction_PreFX}
  n2 = TextureSampleParameter2D() {
   parameterName=Skin FX
   samplerType=SAMPLERTYPE_Masks
   texture=/Game/Characters/Player/Female/Medium/Bodies/F_MED_Bistro_Astronaut/Textures/T_F_MED_Bistro_Astronaut_Body_FX
  }
  n3 = TextureSampleParameter2D() {
   parameterName=TransitionMap
   samplerType=SAMPLERTYPE_Masks
   texture=/Game/Characters/Player/Female/Medium/Bodies/F_MED_Bistro_Astronaut/Textures/TransitionDistance_M
  }
  n4 = ScalarParameter() {parameterName=MaterialCurveB}
  n5 = MaterialFunctionCall!(
   Emissive Alpha=LinearInterpolate(A=@n2.G, B=@n3.R, Alpha=Saturate(None=@n4))
   Local Position
   In=MaterialFunctionCall(Mask=@n2.R, Materials=@n1.Result) {materialFunction=&MF_MetalsLUT}.Result
  ) {materialFunction=&/MF_BistroTransition}
  MaterialFunctionCall(
   AO=@n1.AO
   Distance Blend=@n1.AO
   Fresnel=@n1.AO
   Gmap=@n1.AO
   Pre Skinned Local Position=@n1.AO
   Material=MaterialFunctionCall!(
    Fresnel=@n1.Fresnel
    Alpha=@n2.B
    pre Skinned=@n1.Fresnel
    In=MaterialFunctionCall!(In=@n5.Result) {materialFunction=/Game/Weapons/WeaponSkins/MaterialFunction/MF_EmissiveFresnel}.Result
   ) {materialFunction=&/MF_RayGunMagic}.Result
  ) {materialFunction=&BaseFN_CharacterFunction_PostFX}
 ]
}
